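; Part modifier for the BlockStack back bling, which plays reactive particle and material effects for eliminations, health or shield gains, chest opens and wins.
; Stat changes, pawn health watchers and DBNO toggle Idle VFX parameters, with cached percentages detecting increases.
; SetMaterialReactivity stamps an effect index and game time on mesh materials, ignoring calls while animating.
; In the front end, OnPostResIn instead previews each reaction in one timed sequence.
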(event EventOnStatChanged
  (params (State%20Name :type Name :default "None") (Stat%20Value :type Integer))
  (bind _returnvalue (CallFunction|GetIdleVFXComponent))
  (|CallParentFunction State%20Name Stat%20Value)
  (CallFunction|SetMaterialReactivity 1.0 true)
  (Effects|Components|ParticleSystem|SetBoolParameter _returnvalue "bElimination" true)
  (Utilities|FlowControl|Delay (Variables|Default|GetReactivityLength))
  (Effects|Components|ParticleSystem|SetBoolParameter _returnvalue "bElimination" false))

(event EventOnPostResIn
  (bind _returnvalue (CallFunction|GetIdleVFXComponent))
  (bind _returnvalue_2 (CallFunction|GetIdleVFXComponent))
  (bind _returnvalue_3 (CallFunction|GetIdleVFXComponent))
  (bind _returnvalue_4 (CallFunction|GetIdleVFXComponent))
  (bind _returnvalue_5 (CallFunction|GetIdleVFXComponent))
  (bind _returnvalue_6 (CallFunction|GetIdleVFXComponent))
  (bind _returnvalue_7 (CallFunction|GetIdleVFXComponent))
  (bind _returnvalue_8 (CallFunction|GetIdleVFXComponent))
  (|CallParentFunction)
  (if (CustomCosmeticModifier|IsinFrontEnd)
    (Utilities|FlowControl|Delay 1.0)
    ; Kill
    (CallFunction|SetMaterialReactivity 1.0 true)
    (Effects|Components|ParticleSystem|SetBoolParameter _returnvalue_7 "bElimination" true)
    (Utilities|FlowControl|Delay (Variables|Default|GetReactivityLength))
    (Effects|Components|ParticleSystem|SetBoolParameter _returnvalue_7 "bElimination" false)
    (Utilities|FlowControl|Delay (Variables|Default|GetFELoopDelay))
    ; Shield
    (Effects|Components|ParticleSystem|SetBoolParameter _returnvalue_8 "bShieldIncreased" true)
    (CallFunction|SetMaterialReactivity 2.0 true)
    (Utilities|FlowControl|Delay (Variables|Default|GetReactivityLength))
    (Effects|Components|ParticleSystem|SetBoolParameter _returnvalue_8 "bShieldIncreased" false)
    (Utilities|FlowControl|Delay (Variables|Default|GetFELoopDelay))
    ; Chest
    (CallFunction|SetMaterialReactivity 3.0 true)
    (Effects|Components|ParticleSystem|SetBoolParameter _returnvalue_4 "bChestOpen" true)
    (Utilities|FlowControl|Delay (Variables|Default|GetReactivityLength))
    (Effects|Components|ParticleSystem|SetBoolParameter _returnvalue_4 "bChestOpen" false)
    (Utilities|FlowControl|Delay (Variables|Default|GetFELoopDelay))
    ; Health
    (Effects|Components|ParticleSystem|SetBoolParameter _returnvalue_2 "bHealtIncreased" true)
    (CallFunction|SetMaterialReactivity 4.0 true)
    (Utilities|FlowControl|Delay (Variables|Default|GetReactivityLength))
    (Effects|Components|ParticleSystem|SetBoolParameter _returnvalue_2 "bHealtIncreased" false)
    (Utilities|FlowControl|Delay (Variables|Default|GetFELoopDelay))
    ; Win
    (CallFunction|SetMaterialReactivity 5.0 true)
    (Effects|Components|ParticleSystem|SetBoolParameter _returnvalue_3 "bVictory" true)
    (Utilities|FlowControl|Delay (Variables|Default|GetReactivityLength))
    (Effects|Components|ParticleSystem|SetBoolParameter _returnvalue_3 "bVictory" false)
    (Utilities|FlowControl|Delay (Variables|Default|GetFELoopDelay))
    (else
      (Utilities|FlowControl|Sequence
        (:then_0
          ;[0.01985 0.15 0.09753 0.5] Healt and Shield reactivity
          (bind (_health _maxhealth _shield _maxshield) (Cosmetics|WatchPawnHealth (CustomCharacter|GetAssociatedPlayerPawn))
            (:then
              (Effects|Components|ParticleSystem|SetFloatParameter _returnvalue "HealthPercentage" 1.0)
              (Effects|Components|ParticleSystem|SetFloatParameter _returnvalue "ShieldPercentage" 0.0)
              (Variables|Default|SetHealthPercentage 1.0)
              (Variables|Default|SetShieldPercentage 0.0))
            (:OnHitPointsChanged
              (Effects|Components|ParticleSystem|SetFloatParameter _returnvalue "HealthPercentage" (Math|Float|SafeDivide _health _maxhealth))
              (Effects|Components|ParticleSystem|SetFloatParameter _returnvalue "ShieldPercentage" (Math|Float|SafeDivide _shield _maxshield))
              ;;[0.1484 0.06433 0.15 0.5] Healt/Shield Increased check
              (Utilities|FlowControl|Sequence
                (:then_0
                  (if (> (Math|Float|SafeDivide _health _maxhealth) (Variables|Default|GetHealthPercentage))
                    (Effects|Components|ParticleSystem|SetBoolParameter _returnvalue "bHealtIncreased" true)
                    (CallFunction|SetMaterialReactivity 4.0 true)
                    (Utilities|FlowControl|Delay (Variables|Default|GetReactivityLength))
                    (Variables|Default|SetHealthPercentage (Math|Float|SafeDivide _health _maxhealth))
                    (Effects|Components|ParticleSystem|SetBoolParameter _returnvalue "bHealtIncreased" false)
                    (else
                      (Variables|Default|SetHealthPercentage (Math|Float|SafeDivide _health _maxhealth)))))
                (:then_1
                  (if (> (Math|Float|SafeDivide _shield _maxshield) (Variables|Default|GetShieldPercentage))
                    (Effects|Components|ParticleSystem|SetBoolParameter _returnvalue "bShieldIncreased" true)
                    (CallFunction|SetMaterialReactivity 2.0 true)
                    (Utilities|FlowControl|Delay (Variables|Default|GetReactivityLength))
                    (Variables|Default|SetShieldPercentage (Math|Float|SafeDivide _shield _maxshield))
                    (Effects|Components|ParticleSystem|SetBoolParameter _returnvalue "bShieldIncreased" false)
                    (else
                      (Variables|Default|SetShieldPercentage (Math|Float|SafeDivide _shield _maxshield)))))))))
        (:then_1
          ;[0.15 0.03572 0.01057 0.5] Open Chest reactivity
          (Cosmetics|WatchCosmeticStat (CustomCharacter|GetAssociatedPlayerPawn) "(TagName=\"StatManager.OpenChests\")" 0 false
            (:OnStatChanged
              (CallFunction|SetMaterialReactivity 3.0 true)
              (Effects|Components|ParticleSystem|SetBoolParameter _returnvalue_5 "bChestOpen" true)
              (Utilities|FlowControl|Delay (Variables|Default|GetReactivityLength))
              (Effects|Components|ParticleSystem|SetBoolParameter _returnvalue_5 "bChestOpen" false))))
        (:then_2
          ;[0.15 0.1367 0 0.5] Win reactivity
          (Cosmetics|WatchCosmeticStat (CustomCharacter|GetAssociatedPlayerPawn) "(TagName=\"StatManager.SeasonWins\")" 0 false
            (:OnStatChanged
              (CallFunction|SetMaterialReactivity 5.0 true)
              (Effects|Components|ParticleSystem|SetBoolParameter _returnvalue_6 "bVictory" true)
              (Utilities|FlowControl|Delay (Variables|Default|GetReactivityLength))
              (Effects|Components|ParticleSystem|SetBoolParameter _returnvalue_6 "bVictory" false))))))))

(event EventPlayerStartDBNO
  (bind _returnvalue (CallFunction|GetIdleVFXComponent))
  ;[0.01985 0.15 0.09753 0.5] Healt and Shield reactivity
  (|CallParentFunction)
  (Effects|Components|ParticleSystem|SetFloatParameter _returnvalue "HealthPercentage" (Math|Float|SafeDivide (stale EventOnPostResIn _health) (stale EventOnPostResIn _maxhealth)))
  (Effects|Components|ParticleSystem|SetFloatParameter _returnvalue "ShieldPercentage" (Math|Float|SafeDivide (stale EventOnPostResIn _shield) (stale EventOnPostResIn _maxshield)))
  ;;[0.1484 0.06433 0.15 0.5] Healt/Shield Increased check
  (Utilities|FlowControl|Sequence
    (:then_0
      (if (> (Math|Float|SafeDivide (stale EventOnPostResIn _health) (stale EventOnPostResIn _maxhealth)) (Variables|Default|GetHealthPercentage))
        (Effects|Components|ParticleSystem|SetBoolParameter _returnvalue "bHealtIncreased" true)
        (CallFunction|SetMaterialReactivity 4.0 true)
        (Utilities|FlowControl|Delay (Variables|Default|GetReactivityLength))
        (Variables|Default|SetHealthPercentage (Math|Float|SafeDivide (stale EventOnPostResIn _health) (stale EventOnPostResIn _maxhealth)))
        (Effects|Components|ParticleSystem|SetBoolParameter _returnvalue "bHealtIncreased" false)
        (else
          (Variables|Default|SetHealthPercentage (Math|Float|SafeDivide (stale EventOnPostResIn _health) (stale EventOnPostResIn _maxhealth))))))
    (:then_1
      (if (> (Math|Float|SafeDivide (stale EventOnPostResIn _shield) (stale EventOnPostResIn _maxshield)) (Variables|Default|GetShieldPercentage))
        (Effects|Components|ParticleSystem|SetBoolParameter _returnvalue "bShieldIncreased" true)
        (CallFunction|SetMaterialReactivity 2.0 true)
        (Utilities|FlowControl|Delay (Variables|Default|GetReactivityLength))
        (Variables|Default|SetShieldPercentage (Math|Float|SafeDivide (stale EventOnPostResIn _shield) (stale EventOnPostResIn _maxshield)))
        (Effects|Components|ParticleSystem|SetBoolParameter _returnvalue "bShieldIncreased" false)
        (else
          (Variables|Default|SetShieldPercentage (Math|Float|SafeDivide (stale EventOnPostResIn _shield) (stale EventOnPostResIn _maxshield))))))))

(event Custom|SetMaterialReactivity
  (params (ReactivityIndex :type Float%20%28double-precision%29) (IsIntro :type Boolean))
  (bind _returnvalue (CustomCharacter|GetSkeletalMeshForPartOnPlayerPawn))
  (if (not (Variables|Default|GetIsAnimatingMaterial))
    (Variables|Default|SetIsAnimatingMaterial true)
    (Rendering|Material|SetScalarParameterValueonMaterials _returnvalue "ReactivityIndex" ReactivityIndex)
    (Rendering|Material|SetScalarParameterValueonMaterials _returnvalue "Timestamp" (Utilities|Time|GetGameTimeinSeconds))
    (Utilities|FlowControl|Delay (Variables|Default|GetReactivityLength))
    (Variables|Default|SetIsAnimatingMaterial false)))

(fn ConstructionScript
  (|CallParentFunction))
