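# Niagara effect for the BlockStack backpack cosmetic, showing glowing stacked blocks with cube and pixel bursts in one of five reactive colors.
# System.ReactivityPicker selects every emitter's color; user variables expose gameplay events like health, shield, chest and elimination.
# System.Timeline.Time drives curves that fade the Idle meshes, lights and LightRays and briefly spike Cubes and PixelSprites spawning.
# Front-end modifiers lower Idle and LightRays opacity outside gameplay.
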NiagaraSystem /BRCosmetics/Accessories/FORT_Backpacks/Backpack_M_MED_BlockStack/FX/NS_Backpack_BlockStack
dependencies {
 usedDataInterfaces [&NiagaraDataInterfaceCurve, &NiagaraDataInterfaceArrayFloat]
 usedDynamicInputs [
  &SimulationPosition
  &Boolean_OrOperation
  &MakeCustomFloatFromBool
  &MakeCustomIntFromBool
 ]
 usedModules [
  &SystemState
  &self:SystemSpawnScript.SystemScriptSource.SystemScriptGraph.NiagaraNodeAssignment_0.SetVariables_#1
  &self:SystemSpawnScript.SystemScriptSource.SystemScriptGraph.NiagaraNodeAssignment_3.SetVariables_#2
  /Game/Effects/Niagara/Modules/Logic/UpdateValueStateChange
  &Timeline
  &self:SystemSpawnScript.SystemScriptSource.SystemScriptGraph.NiagaraNodeAssignment_2.SetVariables_#3
  &self:SystemSpawnScript.SystemScriptSource.SystemScriptGraph.NiagaraNodeAssignment_1.SetVariables_#4
 ]
}
emitters {
 Cubes {
  emitterUpdateScript [
   &EmitterState(Life Cycle Mode=::NewEnumerator1, Loop Behavior=::NewEnumerator1, Loop Duration Mode=::NewEnumerator1, UseLoopDelay=false)
   &SpawnRate(
    SpawnRate=&FloatFromCurve(
     FloatCurve=DataInterface() {propertyValues {Curve {keys [<at 0.075 0>,<at 0.105 1>,<at 0.13 0>]}}}
     CurveIndex=$System.Timeline.Time
     Scale Curve=100
    )
    Use Spawn Probability=false
   )
  ]
  particleSpawnScript [
   &InitializeParticle(
    Color=<rgba 1 1 1 2>
    Lifetime Max=0.45
    Lifetime Min=0.35
    Lifetime Mode=::NewEnumerator1
    Mesh Scale Mode=::NewEnumerator1
    Mesh Uniform Scale Max=1.25
    Mesh Uniform Scale Min=0.7
    Position Offset=<0 -12.0461 16.4314>
    Sprite Size Mode=::NewEnumerator3
    Uniform Sprite Size=8
    UsePositionOffset=true
    Write Material Random=true
   )
   &ShapeLocation(MaskSpawnGroup=false, Override Seed=false, Sphere Radius=4, UseOwnerScale=true)
   &InitialMeshOrientation(Use Rotation Vector=false)
   &AddVelocity(Module.Use Falloff Distance=false, Origin Offset=<0 -12.0461 13.2>, Override Seed=false, Velocity Mode=::NewEnumerator1, Velocity Speed=140)
  ]
  particleUpdateScript [
   &ParticleState()
   &DynamicMaterialParameters(
    Index 0 Param 2=1
    Index 0 Param 3=1
    Module.Index0Float0Write=true
    Module.Index0Float1Write=true
    Module.Index0Float2Write=true
    Module.Index0Float3Write=true
   )
   &CurlNoiseForce(
    Bias Noise=false
    Noise Frequency=15
    Noise Quality / Cost=::NewEnumerator0
    Noise Strength=&RandomRangeFloat(Minimum=30, Maximum=100)
    Pan Noise=false
   )
   &Drag(Drag=4.0451, Use Linear Drag=true, Use Rotational Drag=true)
   &GravityForce(Gravity=&RandomRangeVector(Minimum=<0 0 -360>, Maximum=<0 0 -20>))
   &ScaleMeshSize(Scale Factor=&VectorFromFloat(Value=&FloatFromCurve(
      FloatCurve=DataInterface() {propertyValues {Curve {keys [
          {arriveTangent=0.3821, interpMode=RCIM_Cubic, leaveTangent=3.4208, tangentMode=RCTM_Break, time=0, value=0}
          {arriveTangentWeight=0.1606, interpMode=RCIM_Cubic, leaveTangentWeight=0.1355, tangentMode=RCTM_Break, time=0.45, value=1}
          {arriveTangentWeight=0.1503, interpMode=RCIM_Cubic, leaveTangentWeight=0.0859, tangentMode=RCTM_User, time=0.7, value=1}
          {arriveTangent=-0.2843, tangentMode=RCTM_User, time=1, value=0}
         ]}}}
      Scale Curve=0.6
     )))
   &UpdateMeshOrientation(Rotation Rate=&RandomRangeFloat(Minimum=0.4, Maximum=1), Rotation Vector=<1 1 1>)
   &ScaleSpriteSize(
    Scale Sprite Size Mode=::NewEnumerator0
    Uniform Scale Factor=&MakeFloatFromVector(Vector=$Particles.Scale, Channel=::NewEnumerator0)
   )
   &ScaleColor(
    Scale Alpha=&FloatFromCurve(FloatCurve=DataInterface() {propertyValues {Curve {keys [{interpMode=RCIM_Cubic, tangentMode=RCTM_Break, time=0.3, value=1}, <at 1 0>]}}})
    Scale RGB=&MakeVectorFromLinearColorRGB(Color=&MergeHelper_ColorSwitch(
      0=<rgba 0.0345 1 0 1>
      1=<rgba 0 0.7268 1 1>
      2=<rgba 1 0.3685 0 1>
      3=<rgba 1 0.7007 0.0187 1>
      4=<rgba 1 0 0.0172 1>
      Index=$System.ReactivityPicker
      Num=::NewEnumerator3
     ))
    ScaleA=true
    ScaleRGB=true
   )
   /SetVariables(Particles.GlowColor=&Multiply_LinearColorByFloat(
     Linear Color=&MergeHelper_ColorSwitch(
      0=<rgba 0.0999 1 0.0677 1>
      1=<rgba 0.0781 0.7481 1 1>
      2=<rgba 1 0.4113 0.0677 1>
      3=<rgba 1 0.7157 0.0677 1>
      4=<rgba 1 0.0677 0.0837 1>
      Index=$System.ReactivityPicker
      Num=::NewEnumerator3
     )
     Float=0.8
    )) {isSetParameters=true}
   &SolveForcesAndVelocity(Clamp Velocity=false, Limit Acceleration=false)
  ]
  rendererClasses [&NiagaraSpriteRendererProperties, &NiagaraMeshRendererProperties]
  renderers [
   {rendererClass=&NiagaraSpriteRendererProperties}
   {rendererClass=&NiagaraMeshRendererProperties, rendererIndex=1}
  ]
 }
 Idle {
  emitterUpdateScript [
   &EmitterState(Life Cycle Mode=::NewEnumerator1, Loop Behavior=::NewEnumerator1, Loop Duration Mode=::NewEnumerator1, UseLoopDelay=false)
   &SpawnBurst_Instantaneous(Spawn Count=3, Use Spawn Probability=false, UseLoopCountLimit=false)
  ]
  particleSpawnScript [
   &FN_SetSpawnGroup(Number of Spawn Groups=3)
   &InitializeParticle(
    Mesh Scale=&Multiply_VectorByFloat(Vector=<0.1 0.1 0.042>, Float=1.05)
    Mesh Scale Mode=::NewEnumerator4
    Position Offset=&MergeHelper_VectorSwitch(
     0=<0 -12.0461 9.5>
     1=<0 -12.0461 16.4314>
     2=<0 -12.0461 20.4701>
     3=?
     Index=&SpawnGroup()
     Num=::NewEnumerator1
    )
    Sprite Size Mode=::NewEnumerator3
    Uniform Sprite Size=36.4933
    UsePositionOffset=true
    Write Material Random=true
   )
   /SetVariables(Particles.VisibilityTag=&SpawnGroup()) {isSetParameters=true}
  ]
  particleUpdateScript [
   &ParticleState(Kill Particles When Lifetime Has Elapsed=false)
   &FN_Light_Attributes(
    Color=&MergeHelper_ColorSwitch(
     0=<rgba 0.0345 1 0 1>
     1=<rgba 0 0.7268 1 1>
     2=<rgba 1 0.3685 0 1>
     3=<rgba 1 0.7007 0.0187 1>
     4=<rgba 1 0 0.0172 1>
     Index=$System.ReactivityPicker
     Num=::NewEnumerator3
    )
    Create Color Attribute=true
    Create Exponent Attribute=true
    Create Position Attribute=true
    Create Radius Attribute=true
    Exponent=3
    Position=&AddVectorToPosition(Position=$Particles.Position, Vector=<0 0 0>)
    Radius=&Multiply_Float(
     A=22
     B=&FloatFromCurve(
      FloatCurve=DataInterface() {propertyValues {Curve {keys [
          {arriveTangent=0.7861, interpMode=RCIM_Cubic, leaveTangent=0.7861, tangentMode=RCTM_User, time=0.02, value=0}
          {arriveTangent=0.1339, interpMode=RCIM_Cubic, leaveTangent=0.1339, tangentMode=RCTM_User, time=0.05, value=1}
          {arriveTangent=0.0363, interpMode=RCIM_Cubic, leaveTangent=0.0363, tangentMode=RCTM_User, time=0.9488, value=1}
          {arriveTangent=0.7861, interpMode=RCIM_Cubic, leaveTangent=0.7861, tangentMode=RCTM_User, time=1, value=0}
         ]}}}
      CurveIndex=$System.Timeline.Time
     )
    )
    Scale Color=&Multiply_Float(
     A=9
     B=&FloatFromCurve(
      FloatCurve=DataInterface() {propertyValues {Curve {keys [
          {arriveTangent=0.7861, interpMode=RCIM_Cubic, leaveTangent=0.7861, tangentMode=RCTM_User, time=0.07, value=0}
          {arriveTangent=0.1339, interpMode=RCIM_Cubic, leaveTangent=0.1339, tangentMode=RCTM_User, time=0.2, value=1}
          {arriveTangent=0.0363, interpMode=RCIM_Cubic, leaveTangent=0.0363, tangentMode=RCTM_User, time=0.85, value=1}
          {arriveTangent=0.7861, interpMode=RCIM_Cubic, leaveTangent=0.7861, tangentMode=RCTM_User, time=1, value=0}
         ]}}}
      CurveIndex=$System.Timeline.Time
     )
    )
   )
   /SetVariables(Particles.SpriteCameraOffset=2) {isSetParameters=true}
   &ScaleColor(
    Scale Alpha=&FloatFromCurve(
     FloatCurve=DataInterface() {propertyValues {Curve {keys [
         {arriveTangent=0.7861, interpMode=RCIM_Cubic, leaveTangent=0.7861, tangentMode=RCTM_User, time=0.07, value=0}
         {arriveTangent=0.1339, interpMode=RCIM_Cubic, leaveTangent=0.1339, tangentMode=RCTM_User, time=0.2, value=1}
         {arriveTangent=0.0363, interpMode=RCIM_Cubic, leaveTangent=0.0363, tangentMode=RCTM_User, time=0.85, value=1}
         {arriveTangent=0.7861, interpMode=RCIM_Cubic, leaveTangent=0.7861, tangentMode=RCTM_User, time=1, value=0}
        ]}}}
     CurveIndex=$System.Timeline.Time
     Scale Curve=&FN_FrontEndFloatModifier(Gameplay=4, Front End=2)
    )
    Scale RGB=&MakeVectorFromLinearColorRGB(Color=&MergeHelper_ColorSwitch(
      0=<rgba 0.0345 1 0 1>
      1=<rgba 0 0.5087 0.7 1>
      2=<rgba 1 0.3685 0 1>
      3=<rgba 1 0.7007 0.0187 1>
      4=<rgba 1 0 0.0172 1>
      Index=$System.ReactivityPicker
      Num=::NewEnumerator3
     ))
    ScaleA=true
    ScaleRGB=true
   )
   /SetVariables(Particles.SpriteColor=&MakeLinearColorFromVectorAndFloat(
     "Vector (RGB)"=&MakeVectorFromLinearColorRGB(Color=&MergeHelper_ColorSwitch(
       0=<rgba 0.0345 1 0 1>
       1=<rgba 0 0.5087 0.7 1>
       2=<rgba 1 0.3685 0 1>
       3=<rgba 1 0.7007 0.0187 1>
       4=<rgba 1 0 0.0172 1>
       Index=$System.ReactivityPicker
       Num=::NewEnumerator3
      ))
     "Float (Alpha)"=&FloatFromCurve(
      FloatCurve=DataInterface() {propertyValues {Curve {keys [
          {arriveTangent=0.7861, interpMode=RCIM_Cubic, leaveTangent=0.7861, tangentMode=RCTM_User, time=0.07, value=0}
          {arriveTangent=0.1339, interpMode=RCIM_Cubic, leaveTangent=0.1339, tangentMode=RCTM_User, time=0.2, value=1}
          {arriveTangent=0.0363, interpMode=RCIM_Cubic, leaveTangent=0.0363, tangentMode=RCTM_User, time=0.85, value=1}
          {arriveTangent=0.7861, interpMode=RCIM_Cubic, leaveTangent=0.7861, tangentMode=RCTM_User, time=1, value=0}
         ]}}}
      CurveIndex=$System.Timeline.Time
      Scale Curve=&Multiply_Float(A=0.7, B=&FN_FrontEndFloatModifier(Gameplay=1, Front End=0.6))
     )
    )) {isSetParameters=true}
  ]
  rendererClasses [
   &NiagaraSpriteRendererProperties
   &NiagaraLightRendererProperties
   &NiagaraMeshRendererProperties
  ]
  renderers [
   {rendererClass=&NiagaraSpriteRendererProperties}
   {rendererClass=&NiagaraLightRendererProperties, rendererIndex=1}
   {rendererClass=&NiagaraMeshRendererProperties, rendererIndex=2}
  ]
 }
 LightRays {
  emitterUpdateScript [
   &EmitterState(Life Cycle Mode=::NewEnumerator1, Loop Behavior=::NewEnumerator1, Loop Duration=1, UseLoopDelay=false)
   &SpawnBurst_Instantaneous(Spawn Group=1, Use Spawn Probability=false, UseLoopCountLimit=false) {module=SpawnBurst_Instantaneous001}
   &SpawnBurst_Instantaneous(Spawn Count=1, Use Spawn Probability=false)
  ]
  particleSpawnScript [
   &InitializeParticle(
    Lifetime=1.9
    Position Offset=<0 -12.0461 7.1328>
    Sprite Size=&MaskVector2dBySpawnGroup(
     Masked Vector2D=&UniformRangedVector2D(Minimum=<80 60>, Maximum=<120 60>)
     Passthrough Vector2D=&UniformRangedVector2D(Minimum=<55 30>, Maximum=<70 30>)
     Spawn Group Mask=0
    )
    Sprite Size Mode=::NewEnumerator4
    UsePositionOffset=true
    Write Material Random=true
   ) {module=InitializeParticle001}
   /Niagara/Modules/Update/SubUV/SubUVAnimation(
    Random SubUV Lookup Index:NiagaraFloat=&FloatFromCurve(
     FloatCurve=DataInterface() {propertyValues {Curve {keys [<at 0 0>,<at 1 1>]}}}
     CurveIndex=&RandomRangeFloat(Minimum=0, Maximum=1)
    )
    SubUV Animation Mode=::NewEnumerator1
   )
  ]
  particleUpdateScript [
   &ParticleState(Kill Particles When Lifetime Has Elapsed=false)
   &ScaleSpriteSize(Scale Factor=&Vector2DFromFloat(Value=&FloatFromCurve(FloatCurve=DataInterface() {propertyValues {Curve {keys [{arriveTangent=1.1486, time=0.0857, value=1}]}}}, Scale Curve=0.416)))
   &ScaleColor(
    Scale Alpha=&FloatFromCurve(
     FloatCurve=DataInterface() {propertyValues {Curve {keys [
         <at 0.07 0 RCIM_Cubic>
         <at 0.2 1>
         {arriveTangent=-1.25, interpMode=RCIM_Cubic, leaveTangent=-1.25, time=0.8307, value=1}
         <at 1 0 RCIM_Cubic>
        ]}}}
     CurveIndex=$System.Timeline.Time
     Scale Curve=&FN_FrontEndFloatModifier(Gameplay=0.6, Front End=0.3)
    )
    Scale RGB=&MakeVectorFromLinearColorRGB(Color=&MaskLinearColorBySpawnGroup(
      Masked Linear Color=&MergeHelper_ColorSwitch(
       0=<rgba 0.3205 2 0.2604 1>
       1=<rgba 0.1422 0.9836 1.3 1>
       2=<rgba 2 0.9475 0.3333 1>
       3=<rgba 2 1.5044 0.375 1>
       4=<rgba 2 0.2708 0.3005 1>
       Index=$System.ReactivityPicker
       Num=::NewEnumerator3
      )
      Passthrough Linear Color=&MergeHelper_ColorSwitch(
       0=<rgba 0.0345 1 0 1>
       1=<rgba 0 0.5087 0.7 1>
       2=<rgba 1 0.3685 0 1>
       3=<rgba 1 0.7007 0.0187 1>
       4=<rgba 1 0 0.0172 1>
       Index=$System.ReactivityPicker
       Num=::NewEnumerator3
      )
     ))
    ScaleA=true
    ScaleRGB=true
   )
  ]
  rendererClasses [&NiagaraSpriteRendererProperties]
  renderers [{rendererClass=&NiagaraSpriteRendererProperties}]
 }
 PixelSprites {
  emitterUpdateScript [
   &EmitterState(
    Enable Distance Culling=true
    Max Distance Response=::KillImmediately
    Module.bUseMaxDistance=true
    Module.bUseMinDistance=false
    Scalability Mode=::NewEnumerator1
   )
   &SpawnRate(
    SpawnRate=&FloatFromCurve(
     FloatCurve=DataInterface() {propertyValues {Curve {keys [<at 0.07 0>,<at 0.1 1>,<at 0.35 0>]}}}
     CurveIndex=$System.Timeline.Time
     Scale Curve=25
    )
    Use Spawn Probability=false
   )
  ]
  particleSpawnScript [
   &FN_SetSpawnGroup(Number of Spawn Groups=2)
   &InitializeParticle(
    Lifetime=&RandomRangeFloat(Minimum=0.5, Maximum=0.8)
    Position Offset=<0 -12.0461 22>
    Sprite Rotation Angle=&SelectFloatFromArray(
     Array Sampling Mode=::NewEnumerator0
     Direct Array Index=?
     Float Selection Array=DataInterface() {propertyValues {FloatData [0,90,180,270]}}
     Random Seed=0
    )
    Sprite Rotation Mode=::NewEnumerator0
    Sprite Size Mode=::NewEnumerator1
    Uniform Sprite Size Max=8
    UsePositionOffset=true
    Write Material Random=true
   )
   &CameraOffset(Camera Offset Amount=20)
   &ShapeLocation(MaskSpawnGroup=false, Override Seed=false, Sphere Radius=14, Sphere Surface Distribution=0.8, UseOwnerScale=true)
   &ApplyInitialForces()
  ]
  particleUpdateScript [
   &ParticleState()
   &ScaleSpriteSize(
    Uniform Curve Scale=0.85
    Uniform Curve Sprite Scale=DataInterface() {propertyValues {Curve {keys [
        {arriveTangent=2.8617, interpMode=RCIM_Cubic, leaveTangent=2.8617, tangentMode=RCTM_User, time=0, value=0}
        {interpMode=RCIM_Cubic, tangentMode=RCTM_User, time=1, value=1}
       ]}}}
   )
   /Niagara/Modules/Update/SubUV/V2/SubUVAnimation(End Frame=4, Start Frame=1)
   &Drag(Drag=5, Use Linear Drag=true, Use Rotational Drag=true)
   &SolveForcesAndVelocity(Clamp Velocity=false, Limit Acceleration=false)
   &ScaleColor(
    Scale Alpha=&FloatFromCurve(
     FloatCurve=DataInterface() {propertyValues {Curve {keys [
         {interpMode=RCIM_Cubic, tangentMode=RCTM_User, time=0, value=1}
         {arriveTangent=-2.5352, interpMode=RCIM_Cubic, leaveTangent=-2.5352, tangentMode=RCTM_User, time=1, value=0}
        ]}}}
     Scale Curve=2
    )
    Scale RGB=&MakeVectorFromLinearColorRGB(Color=&MergeHelper_ColorSwitch(
      0=<rgba 0.0345 1 0 1>
      1=<rgba 0 0.7268 1 1>
      2=<rgba 1 0.3685 0 1>
      3=<rgba 1 0.7007 0.0187 1>
      4=<rgba 1 0 0.0172 1>
      Index=$System.ReactivityPicker
      Num=::NewEnumerator3
     ))
    ScaleA=true
    ScaleRGB=true
   )
   /SetVariables(
    Particles.GlowColor=&MakeLinearColorFromVectorAndFloat(
     "Vector (RGB)"=&MakeVectorFromLinearColorRGB(Color=&Multiply_LinearColorByFloat(
       Linear Color=&MergeHelper_ColorSwitch(
        0=<rgba 0.0999 1 0.0677 1>
        1=<rgba 0.0781 0.7481 1 1>
        2=<rgba 1 0.4113 0.0677 1>
        3=<rgba 1 0.7157 0.0677 1>
        4=<rgba 1 0.0677 0.0837 1>
        Index=$System.ReactivityPicker
        Num=::NewEnumerator3
       )
       Float=2
      ))
     "Float (Alpha)"=&FloatFromCurve(
      FloatCurve=DataInterface() {propertyValues {Curve {keys [{interpMode=RCIM_Cubic, tangentMode=RCTM_User, time=0.5, value=1}, <at 1 0>]}}}
      Scale Curve=0.05
     )
    )
    Particles.GlowSize=&Multiply_Vector2D(A=$Particles.SpriteSize, B=<2.5 2.5>)
   ) {isSetParameters=true}
  ]
  rendererClasses [&NiagaraSpriteRendererProperties]
  renderers [
   {rendererClass=&NiagaraSpriteRendererProperties}
   {rendererClass=&NiagaraSpriteRendererProperties, rendererIndex=1}
  ]
 }
}
summary {
 systemName=NS_Backpack_BlockStack
 emitters [
  {
   emitterName=Idle
   rendererClasses [
    &NiagaraSpriteRendererProperties
    &NiagaraLightRendererProperties
    &NiagaraMeshRendererProperties
   ]
  }
  {emitterName=Cubes, rendererClasses [&NiagaraSpriteRendererProperties, &NiagaraMeshRendererProperties]}
  {emitterName=LightRays, rendererClasses [&NiagaraSpriteRendererProperties]}
  {emitterName=PixelSprites, rendererClasses [&NiagaraSpriteRendererProperties]}
 ]
 userVariables [
  User.bChestOpen {type=NiagaraBool}
  User.bElimination {type=NiagaraBool}
  User.bHealtIncreased {type=NiagaraBool}
  User.bisFrontEnd {type=NiagaraBool}
  User.bIsFrontEndPreview {type=NiagaraBool}
  User.bIsSwinging {type=NiagaraBool}
  User.bShieldIncreased {type=NiagaraBool}
  User.bVictory {type=NiagaraBool}
  User.DebugReactivity {type=NiagaraBool}
  User.HealthPercentage {type=NiagaraFloat, defaultValue=1, description=""}
  User.KillCount {type=NiagaraInt32}
  User.ShieldPercentage {type=NiagaraFloat, defaultValue=0, description=""}
 ]
}
