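# Material function that layers an animated hologram effect onto input material attributes for the BlockStack backpack, overriding emissive, world position offset, opacity and opacity mask.
# Noise-driven glitches with sine flicker offset vertices camera-relative and shift hue.
# Stabilized screen-space scanlines and flowing stroke bands add detail, with DisableStrokeFX removing stroke glow and displacement.
# Stroke displacement is zero on forward and mobile paths, while opacity uses temporal dithering and sorted translucency.
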MaterialFunction /BRCosmetics/Accessories/FORT_Backpacks/Backpack_M_MED_BlockStack/Materials/MF_BlockStack_Backpack_Hologram
graph {nodes [
  n1 = GetMaterialAttributes(None=NamedRerouteUsage()) {attributeGetTypes [#1, #2, #3]}
  n2 = SetMaterialAttributes!(
   MaterialAttributes=@n1.MaterialAttributes
   EmissiveColor=NamedRerouteUsage()
   WorldPositionOffset=NamedRerouteUsage()
   Opacity=NamedRerouteUsage()
   OpacityMask=NamedRerouteUsage()
  ) {attributeSetTypes [
    #1
    #2
    #4
    #3
   ]}
  FunctionOutput(None=@n2)
  NamedRerouteDeclaration(None=FunctionInput() {inputType=FunctionInput_MaterialAttributes}) {name=InMaterial, nodeColor=<rgba 1 0.5391 0 1>}
  n3 = NamedRerouteUsage()
  n4 = Multiply(
   A=AppendVector(A=Constant(), B=ScalarParameter() {defaultValue=0.15, group=Hologram|Glitch, parameterName=Glitch_Speed})
   B=NamedRerouteUsage()
  )
  n5 = Multiply(
   A=Constant2Vector() {g=1}
   B=Multiply(A=Add(A=@n3, B=@n4), B=ScalarParameter() {defaultValue=15, group=Hologram|Glitch, parameterName=Glitch_Density})
  )
  n6 = Multiply(A=Subtract(
    A=TextureSample(UVs=@n5) {samplerType=SAMPLERTYPE_LinearColor, texture=&T_NoisePack_LinearColor}.R
    B=Constant() {r=0.5}
   ) {constB=0.5}) {constB=2}
  n7 = ScalarParameter() {defaultValue=1, group=Hologram|Glitch, parameterName=Glitch_Subdivision}
  n8 = Ceil(None=Multiply(A=Abs(None=@n6), B=@n7))
  n9 = Add(A=Floor(None=Clamp(None=@n6) {maxDefault=0, minDefault=-1})) {constB=0.5}
  n10 = Multiply(A=Multiply(A=Divide(A=@n8, B=@n7), B=Multiply(A=@n9) {constB=2}), B=NamedRerouteUsage())
  n11 = Transform(None=Constant3Vector() {constant=<rgba 1 0 0 0>}) {transformSourceType=TRANSFORMSOURCE_Camera}
  n12 = ScalarParameter() {defaultValue=1.5, group=Hologram|Glitch, parameterName=Glitch_WPOStrength}
  NamedRerouteDeclaration(None=Multiply(A=@n10, B=Multiply(A=@n11, B=@n12))) {name=Glitch_WpoOutput, nodeColor=<rgba 0 0.1217 1 1>}
  n13 = Constant()
  n14 = ScalarParameter() {defaultValue=0.01, group=Hologram|Scanline, parameterName=Scanline_FlowSpeed}
  n15 = Add(A=NamedRerouteUsage(), B=Multiply(A=AppendVector(A=@n13, B=@n14), B=NamedRerouteUsage()))
  n16 = Frac(None=Multiply(
    A=MaterialFunctionCall(Float2=@n15) {materialFunction=&BreakOutFloat2Components}.G
    B=ScalarParameter() {defaultValue=125, group=Hologram|Scanline, parameterName=Scanline_Density}
   ))
  n17 = Min(A=OneMinus(None=@n16), B=@n16)
  n18 = ScalarParameter() {defaultValue=-0.3968, group=Hologram|Scanline, parameterName=Scanline_ThicknessOffset}
  n19 = Multiply(A=Add(A=Multiply(A=@n17) {constB=2}, B=@n18), B=Divide(B=Add(A=@n18)))
  n20 = Clamp(None=Add(A=ScalarParameter() {defaultValue=0.704, group=Hologram|Scanline, parameterName=Scanline_Sharpness})) {clampMode=CMODE_ClampMin, minDefault=1}
  n21 = Multiply(
   A=Saturate(None=Multiply(A=@n19, B=@n20))
   B=ScalarParameter() {defaultValue=0.75, group=Hologram|Scanline, parameterName=Scanline_ToEmissiveScale}
  )
  NamedRerouteDeclaration(None=Add(A=@n21)) {name=Scanline_ToEmissive, nodeColor=<rgba 0 0.7656 1 1>}
  NamedRerouteDeclaration(None=MaterialFunctionCall!(
    ObjectPivotLocation
    ObjectPositionOffset=VectorParameter() {group=Hologram|Scanline, parameterName=ScanlineOffset, defaultValue=<rgba 0 4.637 0 1>}.RGB
   ) {materialFunction=/Game/Characters/Player/Common/MaterialFunctions/Superlevel/ImprovedStabilizedScreenSpaceUVs}.StabilizedUVs) {name=ScanlineUV, nodeColor=<rgba 1 0.9141 0 1>}
  NamedRerouteDeclaration(None=NamedRerouteUsage()) {name=StrokeUV, nodeColor=<rgba 0 0.3906 1 1>}
  NamedRerouteDeclaration(None=Time() {bOverride_Period=true, period=6000}) {name=Time, nodeColor=<rgba 1 0 0.8493 1>}
  NamedRerouteDeclaration(None=MaterialFunctionCall() {materialFunction=/Game/Tools/3DHB/Shader/Base/MF_3DHB_Base_ScreenUVWithDistanceScale}.Result) {name=GlitchUV, nodeColor=<rgba 1 0.6094 0 1>}
  n22 = NamedRerouteUsage()
  n23 = Multiply(A=ScalarParameter() {defaultValue=1, group=Hologram|Glitch, parameterName=Glitch_FlickerSpeed}, B=@n22)
  n24 = Sine(None=@n23) {period=1.6}
  n25 = NamedRerouteUsage()
  n26 = Add(
   A=Multiply(
    A=MaterialFunctionCall(Float2=@n25) {materialFunction=&BreakOutFloat2Components}.G
    B=ScalarParameter() {defaultValue=0.1, group=Hologram|Glitch, parameterName=Glitch_FlickerUVScale}
   ) {constB=0.25}
   B=@n23
  )
  n27 = ScalarParameter() {defaultValue=0.75, group=Hologram|Glitch, parameterName=Glitch_FlickerFrequency}
  n28 = Add(A=Sine(None=@n26) {period=2.1}, B=Multiply(B=@n27) {constA=0.25})
  n29 = Add(A=Sine(None=@n26) {period=0.72}, B=Multiply(B=@n27) {constA=0.75}) {constB=0.75}
  n30 = Add(A=Multiply(B=@n27) {constA=0.9, constB=0.9}, B=Sine(None=@n26) {period=1.25}) {constB=0.9}
  n31 = Saturate(None=Multiply(A=Multiply(A=@n28, B=@n29), B=Multiply(A=@n29, B=@n30)))
  n32 = LinearInterpolate(
   A=OneMinus(None=ScalarParameter() {defaultValue=0.4, group=Hologram|Glitch, parameterName=Glitch_FlickerDetailShakeStrength})
   Alpha=ConstantBiasScale(None=Sine(None=@n22) {period=0.075})
  ) {constA=0.7}
  NamedRerouteDeclaration(None=Multiply(A=Ceil(None=@n24), B=Multiply(A=@n31, B=@n32))) {name=Flicker, nodeColor=<rgba 0 1 0.5781 1>}
  n33 = Add(A=LocalPosition().XYZ, B=VectorParameter() {group=Hologram, parameterName=PreSkinnedPosition Offset, defaultValue=<rgba 0 0 6.4415 1>}.RGB)
  n34 = Multiply(A=Multiply(A=ComponentMask(None=@n33) {b=true, r=true}, B=Constant2Vector() {g=-1, r=1})) {constB=0.01}
  NamedRerouteDeclaration(None=ShaderStageSwitch(PixelShader=VertexInterpolator(VS=@n34).PS, VertexShader=@n34)) {name=PreSkinnedPos, nodeColor=<rgba 0 0.7656 1 1>}
  NamedRerouteDeclaration(None=ScalarParameter() {defaultValue=1, group=Hologram, parameterName=EnableHologram}) {name=EnableHologram, nodeColor=<rgba 0.2188 0 1 1>}
  NamedRerouteDeclaration(None=@n21) {name=ScanlinesGrayscale, nodeColor=<rgba 1 0.1172 0 1>}
  NamedRerouteDeclaration(None=LinearInterpolate(B=ScalarParameter() {defaultValue=2, group=Hologram|Glitch, parameterName=Glitch_HueshiftStrength}, Alpha=@n10)) {name=Glitch_HueShiftOuput, nodeColor=<rgba 0.6875 0 1 1>}
  n35 = NamedRerouteUsage()
  n36 = NamedRerouteUsage()
  n37 = Multiply(A=ScalarParameter() {defaultValue=7, group=Hologram|Stroke, parameterName=Stroke_FlowSpeed}) {constB=-1}
  n38 = Add(
   A=Multiply(A=MaterialFunctionCall(Float2=@n35) {materialFunction=&BreakOutFloat2Components}.G) {constB=-10}
   B=Add(A=Multiply(A=@n36, B=@n37)) {constB=0}
  )
  n39 = Multiply(A=ScalarParameter() {defaultValue=3, group=Hologram|Stroke, parameterName=Stroke_Frequency}) {constB=0.01}
  n40 = ScalarParameter() {defaultValue=2, group=Hologram|Stroke, parameterName=Stroke_Thickness}
  n41 = OneMinus(None=Saturate(None=Multiply(A=@n39, B=@n40)))
  n42 = Subtract(A=Frac(None=Multiply(A=@n38, B=@n39)), B=@n41)
  n43 = Power(
   Base=MaterialFunctionCall(Input=Saturate(None=@n42), Input Low=Constant(), Input High=Subtract(B=@n41)) {materialFunction=&RemapValue_0-1}.Result
   Exp=ScalarParameter() {defaultValue=2, group=Hologram|Stroke, parameterName=Stroke_GradientTweak}
  )
  n44 = Multiply(A=@n43, B=ScalarParameter() {group=Hologram|Stroke, parameterName=Stroke_WPOStrength})
  NamedRerouteDeclaration(None=Saturate(None=@n44)) {name=StrokeGrayscale, nodeColor=<rgba 0 0.2031 1 1>}
  n45 = Max(A=NamedRerouteUsage(), B=NamedRerouteUsage())
  NamedRerouteDeclaration(None=MaterialFunctionCall(Alpha Threshold=@n45) {materialFunction=&DitherTemporalAA_Fortnite}.Result) {name=OpacityMask, nodeColor=<rgba 0.5234 1 0 1>}
  n46 = ScalarParameter() {defaultValue=0.85, group=Hologram, parameterName=OpacityMin}
  n47 = ScalarParameter() {defaultValue=1, group=Hologram, parameterName=OpacityMax}
  NamedRerouteDeclaration(None=MaterialFunctionCall!(Opacity=LinearInterpolate(A=@n46, B=@n47, Alpha=@n45)) {materialFunction=/Game/Global/Materials/MaterialGym/Materials/SortedOutline/MF_SortTranslucency}.Result) {name=Opacity, nodeColor=<rgba 0.7109 1 0 1>}
  NamedRerouteDeclaration(None=NamedRerouteUsage()) {name=BaseUV, nodeColor=<rgba 0.5234 0 1 1>}
  NamedRerouteDeclaration(None=@n1.EmissiveColor) {name=InEmissive, nodeColor=<rgba 1 0 0.7031 1>}
  NamedRerouteDeclaration(None=@n1.WorldPositionOffset) {name=InWPO, nodeColor=<rgba 0 1 0.9766 1>}
  n48 = NamedRerouteUsage()
  n49 = NamedRerouteUsage()
  n50 = ScalarParameter() {defaultValue=0.8, group=Hologram, parameterName=HologramSaturation}
  n51 = Multiply(
   A=Desaturation(None=Multiply(A=@n48, B=@n49), Fraction=OneMinus(None=@n50))
   B=ScalarParameter() {defaultValue=1.5, group=Hologram, parameterName=HologramIntensity}
  )
  n52 = LinearInterpolate(
   A=@n51
   B=MaterialFunctionCall(Hue Shift Percentage=NamedRerouteUsage(), Texture=@n51) {materialFunction=&HueShift}.Result
   Alpha=ScalarParameter() {defaultValue=0.3, group=Hologram|Glitch, parameterName=GlitchColorBlendAmount}
  )
  n53 = Add(A=NamedRerouteUsage(), B=@n52)
  NamedRerouteDeclaration(None=LinearInterpolate(
    A=NamedRerouteUsage()
    B=StaticSwitchParameter(True=@n52, False=@n53) {group=Hologram|Stroke, parameterName=DisableStrokeFX}
    Alpha=NamedRerouteUsage()
   )) {name=OutEmissive, nodeColor=<rgba 1 0.6563 0 1>}
  n54 = NamedRerouteUsage()
  n55 = NamedRerouteUsage()
  n56 = NamedRerouteUsage()
  n57 = NamedRerouteUsage()
  n58 = SmoothStep(Value=Abs(None=ComponentMask(None=@n57) {r=true})) {constMax=0.1}
  n59 = StaticSwitchParameter(True=@n55, False=Add(A=@n55, B=Multiply(A=@n56, B=@n58))) {group=Hologram|Stroke, parameterName=DisableStrokeFX}
  NamedRerouteDeclaration(None=LinearInterpolate(A=@n54, B=Add(A=@n59, B=@n54), Alpha=NamedRerouteUsage())) {name=OutWPO, nodeColor=<rgba 0 1 0.0156 1>}
  n60 = Constant3Vector() {constant=<rgba -1 0 0 0>}
  n61 = Constant3Vector() {constant=<rgba 1 0 0 0>}
  n62 = MaterialFunctionCall(Float2=NamedRerouteUsage()) {materialFunction=&BreakOutFloat2Components}
  n63 = Transform(None=Clamp(None=LinearInterpolate(A=@n60, B=@n61, Alpha=@n62.R)) {minDefault=-1})
  n64 = Constant()
  NamedRerouteDeclaration(None=ShadingPathSwitch(Default=Multiply(A=@n44, B=@n63), Forward=@n64, Mobile=@n64)) {name=Stroke_WpoOutput, nodeColor=<rgba 1 0 0.5625 1>}
  NamedRerouteDeclaration(None=Multiply(A=VectorParameter() {group=Hologram|Stroke, parameterName=Stroke_Color, defaultValue=<rgba 0 0.4921 0.5 1>}.RGB, B=@n43)) {name=Stroke_ColorOutput, nodeColor=<rgba 0 1 0.6484 1>}
 ]}
